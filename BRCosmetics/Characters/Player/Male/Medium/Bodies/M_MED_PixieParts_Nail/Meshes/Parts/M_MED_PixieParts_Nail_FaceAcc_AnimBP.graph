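# Animation Blueprint for the face accessory part of the PixieParts Nail player character, adding movement-driven motion.
# It copies pose and curves from the attached parent mesh, then feeds them through the part's Control Rig.
# Each update, while the cached Fort pawn is valid, a PlayerMovement value is split by clamped range mapping into left and right amounts.
# Those amounts drive the rig's character move inputs.
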AnimBlueprint /BRCosmetics/Characters/Player/Male/Medium/Bodies/M_MED_PixieParts_Nail/Meshes/Parts/M_MED_PixieParts_Nail_FaceAcc_AnimBP
target_skeleton=&/M_MED_PixieParts_Nail_FaceAcc_Skeleton
graphs [
 AnimGraph {nodes [
   n1 = Control Rig!(
    Source=Copy Pose From Mesh() {bCopyCurves=true, bUseAttachedParent=true}
    Alpha
    character_move_left:double=Get Character Move Left() -> (Character Move Left:double)
    character_move_right:double=Get Character Move Right() -> (Character Move Right:double)
   ) -> (Pose) {controlRigAssetReference {controlRigAsset=&/M_MED_PixieParts_Nail_FaceAcc_CtrlRig}}
   Output Pose(Result=@n1) {layerGroup=None}
  ]}
 EventGraph {nodes [
   Event BlueprintInitializeAnimation()
   > n1 = Cast To FortPlayerPawn(Object=TryGetFortPawn())
   > Set Pawn(Pawn:Fort Player Pawn Object Reference=@n1) -> (Output_Get:Fort Player Pawn Object Reference)
   Event BlueprintUpdateAnimation()
   > Get!(execute:Exec) -> (then:Exec, else:Exec, Pawn:Fort Player Pawn Object Reference)
   > n2 = PlayerMovement(AnimInstance=Self-Reference(), XDotProductCoefficient=25, YDotProductCoefficient=25, ZDotProductCoefficient=1)
   > Set Character Move Right(Character Move Right:double=MapRangeClamped(Value=@n2, InRangeA=0, InRangeB=25)) -> (Output_Get:double)
   > Set Character Move Left(Character Move Left:double=MapRangeClamped(Value=@n2, InRangeA=0, InRangeB=-25)) -> (Output_Get:double)
  ]}
]
variables [
 Body_Mesh {type=SkeletalMeshComponent}
 Character Move Left {type=number}
 Character Move Right {type=number}
 Pawn {type=FortPlayerPawn}
]
